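# Translucent, two-sided lit water material for the Utopia City fountain waterfall, with time-of-day base color, foam color and opacity.
# Two panning normals are blended, while a scrolling displacement texture mixes water with foam and varies roughness.
# Opacity combines vertical fades, noise-broken side edges and Fresnel; normal-map red drives optional vertex-normal displacement.
# Vertex color red offsets panning time and green scales U tiling, varying the flow per vertex.
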Material /Game/Effects/Fort_Effects/Effects/Environment/Hermes/UtopiaCity/Fountain/M_Utopia_Fountain_Waterfall
graph {
 properties {
  blendMode=BLEND_Translucent
  ditherOpacityMask=true
  pixelDepthOffsetMode=PDOM_Legacy
  translucencyLightingMode=TLM_Surface
  twoSided=true
  usedShadingModels=Default Lit
 }
 nodes [
  n1 = VectorParameter() {group=Base Colors, parameterName=BaseColor Morning, sortPriority=0, defaultValue=<rgba 0.0873 0.3024 0.3646 1>}
  n2 = VectorParameter() {group=Base Colors, parameterName=BaseColor Day, sortPriority=1, defaultValue=<rgba 0.0873 0.3024 0.3646 1>}
  n3 = VectorParameter() {group=Base Colors, parameterName=BaseColor Evening, sortPriority=2, defaultValue=<rgba 0.0873 0.3024 0.3646 1>}
  n4 = VectorParameter() {group=Base Colors, parameterName=BaseColor Night, sortPriority=3, defaultValue=<rgba 0.0873 0.3024 0.3646 1>}
  n5 = VectorParameter() {group=Foam Colors, parameterName=FoamColor Morning, sortPriority=0, defaultValue=<rgba 0.5052 0.8768 0.9427 1>}
  n6 = VectorParameter() {group=Foam Colors, parameterName=FoamColor Day, sortPriority=1, defaultValue=<rgba 0.5052 0.8768 0.9427 1>}
  n7 = VectorParameter() {group=Foam Colors, parameterName=FoamColor Evening, sortPriority=2, defaultValue=<rgba 0.5052 0.8768 0.9427 1>}
  n8 = VectorParameter() {group=Foam Colors, parameterName=FoamColor Night, sortPriority=3, defaultValue=<rgba 0.5052 0.8768 0.9427 1>}
  n9 = ScalarParameter() {defaultValue=2, parameterName=Water_N_2_TileX}
  n10 = TextureCoordinate()
  n11 = Multiply(A=NamedRerouteUsage(), B=ComponentMask(None=@n10) {r=true})
  n12 = ComponentMask(None=@n10) {g=true}
  n13 = ScalarParameter() {defaultValue=0.5, parameterName=Water_N_2_TileY}
  n14 = Panner(
   Coordinate=AppendVector(A=Multiply(A=@n9, B=@n11), B=Multiply(A=@n12, B=@n13))
   Time=NamedRerouteUsage()
   Speed=ComponentMask(None=VectorParameter() {parameterName=Normal2Speed, defaultValue=<rgba 0.05 -1 0 0>}.RGB) {g=true, r=true}
  ) {bFractionalPart=true}
  n15 = TextureSample(UVs=@n14) {texture=/Game/Packages/DS_Fortnite_TRV_Terrain/Textures/T_TRV_HF_WaterDisp}
  LinearInterpolate(
   A=MaterialFunctionCall(Color_Morning=@n1.RGB, Color_Day=@n2.RGB, Color_Evening=@n3.RGB, Color_Night=@n4.RGB) {materialFunction=&Color_Time_of_Day}.Result
   B=MaterialFunctionCall(Color_Morning=@n5.RGB, Color_Day=@n6.RGB, Color_Evening=@n7.RGB, Color_Night=@n8.RGB) {materialFunction=&Color_Time_of_Day}.Result
   Alpha=@n15.G
  )
  LinearInterpolate(
   A=ScalarParameter() {defaultValue=0.05, parameterName=Roughness Min}
   B=ScalarParameter() {defaultValue=0.4, parameterName=Roughness Max}
   Alpha=@n15.G
  ) {constA=0.05, constB=0.4}
  n16 = TextureCoordinate()
  n17 = Multiply(A=ScalarParameter() {defaultValue=1, parameterName=Water_N_1_TileX}, B=Multiply(A=NamedRerouteUsage(), B=ComponentMask(None=@n16) {r=true}))
  n18 = Multiply(A=ComponentMask(None=@n16) {g=true}, B=ScalarParameter() {defaultValue=0.8, parameterName=Water_N_1_TileY})
  n19 = VectorParameter() {parameterName=NormalSpeed, defaultValue=<rgba 0.1 -1.5 0 0>}
  n20 = TextureSample(UVs=Panner(Coordinate=AppendVector(A=@n17, B=@n18), Time=NamedRerouteUsage(), Speed=ComponentMask(None=@n19.RGB) {g=true, r=true}) {bFractionalPart=true}) {samplerType=SAMPLERTYPE_Normal, texture=/Game/Packages/DS_Fortnite_TRV_Terrain/Textures/T_TRV_HF_WaterNormal}
  n21 = Constant()
  Multiply(
   A=Multiply(
    A=MaterialFunctionCall(X=@n20.R, Y=@n20.R, Z=@n21) {materialFunction=&MakeFloat3}.Result
    B=ScalarParameter() {parameterName=WPO Scalar}
   )
   B=VertexNormalWS()
  )
  n22 = ComponentMask(None=TextureCoordinate()) {g=true}
  n23 = ScalarParameter() {defaultValue=1, group=Opacity, parameterName=Opacity Morning, sortPriority=0}
  n24 = ScalarParameter() {defaultValue=1, group=Opacity, parameterName=Opacity Day, sortPriority=1}
  n25 = ScalarParameter() {defaultValue=1, group=Opacity, parameterName=Opacity Evening, sortPriority=2}
  n26 = ScalarParameter() {defaultValue=1, group=Opacity, parameterName=Opacity Night, sortPriority=3}
  n27 = Multiply(
   A=Power(Base=OneMinus(None=@n22), Exp=ScalarParameter() {defaultValue=1, parameterName=ScaleMask Power}) {constExponent=1}
   B=Multiply(
    A=MaterialFunctionCall(Value_Morning=@n23, Value_Day=@n24, Value_Evening=@n25, Value_Night=@n26) {materialFunction=&Scalar_Time_of_Day}.Result
    B=ScalarParameter() {defaultValue=8, parameterName=ScaleMask}
   )
  )
  n28 = Power(Base=@n22, Exp=ScalarParameter() {defaultValue=2, parameterName=Top Edge Fade Power})
  n29 = ScalarParameter() {defaultValue=2, parameterName=Top Edge Fade Multi}
  n30 = LinearInterpolate(A=@n27, B=Multiply(A=@n27, B=Multiply(A=@n28, B=@n29)), Alpha=ScalarParameter() {parameterName=Use TopEdgeFade})
  n31 = ComponentMask(None=TextureCoordinate()) {r=true}
  n32 = Multiply(A=Multiply(A=OneMinus(None=@n31), B=@n31)) {constB=4}
  n33 = TextureCoordinate()
  n34 = Multiply(A=ScalarParameter() {defaultValue=3, parameterName=EdgeMask_TileX}, B=Multiply(A=NamedRerouteUsage(), B=ComponentMask(None=@n33) {r=true}))
  n35 = Multiply(A=ComponentMask(None=@n33) {g=true}, B=ScalarParameter() {defaultValue=2, parameterName=EdgeMask})
  n36 = VectorParameter() {parameterName=EdgeMaskSpeed, defaultValue=<rgba 0.05 -2.85 0 0>}
  n37 = TextureSample(UVs=Panner(Coordinate=AppendVector(A=@n34, B=@n35), Time=NamedRerouteUsage(), Speed=ComponentMask(None=@n36.RGB) {g=true, r=true}) {bFractionalPart=true}) {texture=/Game/Global/Textures/Default/Noise/TankMoteNoise}
  n38 = Multiply(A=@n32, B=LinearInterpolate(Alpha=Add(A=@n32, B=@n37.G)) {constA=-0.5})
  n39 = If(A=ScalarParameter() {defaultValue=1, parameterName=Side Gradient Transition}, "A > B"=@n38, "A == B"=@n38, "A < B"=Constant() {r=1}) {constB=0.99}
  n40 = ScalarParameter() {defaultValue=1, parameterName=Fresnel Exponent}
  Multiply(
   A=Clamp(None=Multiply(A=@n30, B=@n39))
   B=LinearInterpolate(A=ScalarParameter() {parameterName=Fresnel Min}, B=ScalarParameter() {parameterName=Fresnel Max}, Alpha=Fresnel(ExponentIn=@n40))
  )
  n41 = ScalarParameter() {defaultValue=1, parameterName=Global Time Multiplier}
  n42 = Time()
  n43 = VertexColor()
  NamedRerouteDeclaration(None=Add(A=Multiply(A=@n41, B=@n42), B=@n43.R)) {name=CustomTime, nodeColor=<rgba 0 0.625 1 1>}
  NamedRerouteDeclaration(None=@n43.G) {name=UScale, nodeColor=<rgba 0.3359 1 0 1>}
  Reroute(None=MaterialFunctionCall(
    BaseNormal=@n20.RGB
    AdditionalNormal=TextureSample(UVs=@n14) {samplerType=SAMPLERTYPE_Normal, texture=/Game/Packages/DS_Fortnite_TRV_Terrain/Textures/T_TRV_SwampWater_01_N}.RGB
   ) {materialFunction=&BlendAngleCorrectedNormals}.Result)
  ScalarParameter() {defaultValue=1, parameterName=Specular}
 ]
}
